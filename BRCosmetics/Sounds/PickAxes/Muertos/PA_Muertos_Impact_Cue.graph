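# Sound cue for the impact of the Muertos pickaxe cosmetic.
# A mixer layers a random pick of four impact waves, with slight pitch and volume variation, over a quieter notes layer with volume variation only.
# The notes selector mixes four note waves with seven empty slots, so that accent plays on only some hits.
# It uses pickaxe impact attenuation and player melee weapon concurrency.
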SoundCue /BRCosmetics/Sounds/PickAxes/Muertos/PA_Muertos_Impact_Cue
root=Mixer_0
nodes [
 {id=Mixer_0, children [Modulator_0,Modulator_1], properties {inputVolume [1,0.65]}}
 {id=Modulator_0, children [Random_0], properties {pitchMax=1.15, pitchMin=1, volumeMin=0.9}}
 {id=Modulator_1, children [Random_1], properties {pitchMax=1, pitchMin=1, volumeMin=0.75}}
 {id=Random_0, children [WavePlayer_0,WavePlayer_1,WavePlayer_2,WavePlayer_4], properties {weights [1,1,1,1]}}
 {
  id=Random_1
  children [WavePlayer_3,null,null,WavePlayer_5,null,WavePlayer_6,null,null,WavePlayer_7,null,null]
  properties {weights [1,1,1,1,1,1,1,1,1,1,1]}
 }
 {id=WavePlayer_0, properties {soundWaveAssetPtr=&/PA_Muertos_Impact_01}}
 {id=WavePlayer_1, properties {soundWaveAssetPtr=&/PA_Muertos_Impact_02}}
 {id=WavePlayer_2, properties {soundWaveAssetPtr=&/PA_Muertos_Impact_03}}
 {id=WavePlayer_3, properties {soundWaveAssetPtr=&/PA_Muertos_Impact_Notes_01}}
 {id=WavePlayer_4, properties {soundWaveAssetPtr=&/PA_Muertos_Impact_04}}
 {id=WavePlayer_5, properties {soundWaveAssetPtr=&/PA_Muertos_Impact_Notes_02}}
 {id=WavePlayer_6, properties {soundWaveAssetPtr=&/PA_Muertos_Impact_Notes_03}}
 {id=WavePlayer_7, properties {soundWaveAssetPtr=&/PA_Muertos_Impact_Notes_04}}
]
properties {
 attenuationSettings=/Game/Sounds/Attenuation/Fort_Pickaxe_Impact_Att
 audioPropertiesSheet=/BRCosmetics/Sounds/Settings/PropertySheets/Pickaxe/APS_Cosmetic_Pickaxe_Impact
 bPrimeOnLoad=true
 maxDistance=4300
 soundClassObject=&Fort_Impacts
 volumeMultiplier=1.2
 attenuationOverrides {
  bAttenuateWithLPF=true
  distanceAlgorithm=Logarithmic
  falloffDistance=800
  lPFRadiusMax=800
  lPFRadiusMin=400
  attenuationShapeExtents {x=300}
 }
 concurrencyOverrides {bLimitToOwner=true, maxCount=2, resolutionRule=StopOldest}
 concurrencySet [&Player_Melee_Weapon_Conc]
}
